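# NS_Blank_Glider is a placeholder Niagara system for gliders in the cosmetics effects content, apparently used where no visible effect is wanted.
# Its single emitter, Empty, holds only emitter state, particle initialization and particle state modules with a sprite renderer.
# It lists no spawn module, so it emits no particles, and it exposes no user parameters.
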NiagaraSystem /BRCosmetics/Effects/Niagara/Systems/Blanks/NS_Blank_Glider
dependencies {usedDataInterfaces [], usedDynamicInputs [], usedModules [&SystemState]}
emitters {Empty {
  emitterUpdateScript [&EmitterState()]
  particleSpawnScript [&InitializeParticle(Color Mode=::NewEnumerator0, Lifetime=5, UsePositionOffset=false, Write Material Random=true)]
  particleUpdateScript [&ParticleState()]
  rendererClasses [&NiagaraSpriteRendererProperties]
  renderers [{rendererClass=&NiagaraSpriteRendererProperties}]
 }}
summary {
 systemName=NS_Blank_Glider
 emitters [{emitterName=Empty, rendererClasses [&NiagaraSpriteRendererProperties]}]
 userVariables []
}
